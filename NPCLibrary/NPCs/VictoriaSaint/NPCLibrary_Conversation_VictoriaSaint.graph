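# Conversation database for the Victoria Saint NPC, defining what she says when a player talks to her.
# A single entry point leads to a speech node with a default line, a first-ever-conversation line, and outfit reactions gated on the player's character cosmetic.
# Contextual lines take their priorities from named curve rows.
# The speech node then links out to the shared NPC Library quests and basic services conversations.
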ConversationDatabase /NPCLibrary/NPCs/VictoriaSaint/NPCLibrary_Conversation_VictoriaSaint
entry_points [{tag=Conversation.Entry.VictoriaSaint, destinations [#1]}]
nodes [
 ConversationEntryPointNode_7 {
  id=#1
  class=ConversationEntryPointNode
  outputs [#2]
  properties {entryTag {tagName=Conversation.Entry.VictoriaSaint}}
 }
 Speech_C_4 {
  id=#2
  class=Speech_C
  outputs [#3, #4]
  properties {
   nodeName=""
   generalConfig {
    defaultMessage=Tell Daraku I'm coming for her. And I'm bringing friends.
    contextualMessages [
     {
      message="Do you know a \"Helsie\"? I got a very glittery letter from her, asking to meet here."
      contextRequirements [{requirement {class=FortControllerRequirement_IsFirstEverConversationWithNPC}}]
      priority {curve {rowName=Tandem.Speech.Priority.FirstEverInteraction}}
     }
     {
      message=I never saw myself as a teacher... But I'm improving upon your lessons.
      contextRequirements [{requirement {class=FortControllerRequirement_HasCID, properties {allowedCIds [/BRCosmetics/Athena/Items/Cosmetics/Characters/Character_LethalVae]}}}]
      priority {curve {rowName=Tandem.Speech.Priority.OutfitReaction}}
     }
     {
      message=It's far too easy to lose yourself in the hunt. Let your friends ground you.
      contextRequirements [{requirement {
         class=FortControllerRequirement_HasCID
         properties {allowedCIds [/BRCosmetics/Athena/Items/Cosmetics/Characters/Character_Silencer, /BRCosmetics/Athena/Items/Cosmetics/Characters/Character_ProdigyFire]}
        }}]
      priority {curve {rowName=Tandem.Speech.Priority.OutfitReaction}}
     }
     {
      message="Have you met Shimmerdusk yet? Don't worry, you will. Soon."
      contextRequirements [{requirement {class=FortControllerRequirement_HasCID, properties {allowedCIds [/BRCosmetics/Athena/Items/Cosmetics/Characters/Character_LazarusLens]}}}]
      priority {curve {rowName=Tandem.Speech.Priority.OutfitReaction}}
     }
     {
      message=Mark my words: I'll be there when you sing your final note.
      contextRequirements [{requirement {class=FortControllerRequirement_HasCID, properties {allowedCIds [/BRCosmetics/Athena/Items/Cosmetics/Characters/Character_IchorIncisor]}}}]
      priority {curve {rowName=Tandem.Speech.Priority.OutfitReaction}}
     }
     {
      message="No... You can't be... Vivica? What happened?!"
      contextRequirements [{requirement {class=FortControllerRequirement_HasCID, properties {allowedCIds [/BRCosmetics/Athena/Items/Cosmetics/Characters/Character_CavalryAlt]}}}]
      priority {curve {rowName=Tandem.Speech.Priority.OutfitReaction}}
     }
    ]
   }
  }
 }
 ConversationLinkNode_0 {id=#4, class=ConversationLinkNode, properties {remoteEntryTag {tagName=Conversation.Entry.BasicServices}}}
 ConversationLinkNode_1 {id=#3, class=ConversationLinkNode, properties {remoteEntryTag {tagName=Conversation.Entry.NPCLibrary.Quests}}}
]
